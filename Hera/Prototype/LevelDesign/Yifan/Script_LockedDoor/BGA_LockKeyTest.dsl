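; Prototype Fortnite lock button that lets a player holding a key item open a referenced BGAMovingDoor.
; On the authority, BeginPlay optionally spawns the key pickup at an arrow after a short delay.
; Interaction requires the key and an unused button, then disables the button, removes the key and opens the door.
; ButtonPressedEvent plays audio on remotes and sets replicated state on the authority, whose OnRep turns the light green.
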(event EventBeginPlay
  ;[0.3757 0.4525 0.8365 0.5] spawn a key in the world if there's no key
  (Variables|BuildingActor|SetbIsInvulnerable true)
  (Fortnite|SetCanbeDamaged self)
  (Networking|SwitchHasAuthority
    (:Authority
      (if (Variables|Default|GetSpawnKey)
        (Utilities|FlowControl|Delay 0.5)
        (Inventory|SpawnPickupInWorldWithLevel (Variables|Default|GetItem_Key) 0 1 (Transformation|GetWorldLocation (Variables|Default|GetArrow))) ; remember to change number to spawn
      ))))

(event Custom|ButtonPressedEvent
  ;[0.244 0.7115 0.11 0.5] Play sound for this player once button is pressed
  (Networking|SwitchHasAuthority
    (:Authority
      (Variables|Default|SetButtonisPressed true))
    (:Remote
      (Audio|Components|Audio|Play (Variables|Default|GetButtonAudio) 0.0))))

(event Interaction|EventBlueprintOnInteract
  (params (InteractingPawn :type Fort%20Pawn%20Object%20Reference) (InteractionBeingAttempted :type EInteractionBeingAttempted%20Enum :default "FirstInteraction"))
  ;[0.7073 0.209 0.644 0.5] On button interaction do these
  (Variables|Default|SetButtonCanbePressed false)
  (CallFunction|RemoveKeyItem InteractingPawn)
  (CallFunction|ButtonPressedEvent)
  (Class|BGAMovingDoor|OpenDoor (Variables|Default|GetDoorReference)))

(fn BlueprintCanInteract
  (params (InteractingPawn :type Fort%20Pawn%20Object%20Reference) (InteractionBeingAttempted :type EInteractionBeingAttempted%20Enum :default "FirstInteraction") (InteractionType :type TInteractionType%20Enum :default "IT_NoInteraction"))
  (bind _asfort_player_pawn_athena (Utilities|Casting|CastToFortPlayerPawnAthena InteractingPawn))
  (bind _asfort_player_controller_athena (Utilities|Casting|CastToFortPlayerControllerAthena (Pawn|GetController _asfort_player_pawn_athena)))
  (return (and (>= (Inventory|GetItemQuantityonInventoryOwner (Fortnite|GetFortPlayerControllerfromActor _asfort_player_controller_athena) (Variables|Default|GetItem_Key)) 1) (Variables|Default|GetButtonCanbePressed))))

(fn BlueprintGetInteractionString
  (params (InteractingPawn :type Fort%20Pawn%20Object%20Reference) (InteractionBeingAttempted :type EInteractionBeingAttempted%20Enum :default "FirstInteraction"))
  (return (Variables|Default|GetInteractText)))

(fn OnRep_ButtonisPressed
  (if (Variables|Default|GetButtonisPressed)
    (Rendering|Components|Light|SetLightColor (Variables|Default|GetPointLight_Red) "(R=0.307666,G=1.000000,B=0.030208,A=1.000000)" true)))

(fn RemoveKeyItem
  (params (Pawn :type Fort%20Pawn%20Object%20Reference))
  (bind _asfort_player_controller_athena (Utilities|Casting|CastToFortPlayerControllerAthena (Pawn|GetController Pawn)))
  (Inventory|RemoveItemfromInventoryOwner _asfort_player_controller_athena (Variables|Default|GetItem_Key)))

(fn ConstructionScript)
