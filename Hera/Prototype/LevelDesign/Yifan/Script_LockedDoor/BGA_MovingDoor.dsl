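; Moving door building actor that slides its mesh downward once when its OpenDoor event is called.
; BeginPlay makes the actor invulnerable and stores the mesh's start location and a lowered end location.
; OpenDoor passes a DoOnce into a timeline that lerps the mesh between those relative locations.
; A timeline event track plays the bunker door opening sound only on the non-authority (remote) side.
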(event EventBeginPlay
  (bind _doormesh (Variables|Default|GetDoorMesh))
  (bind _relativelocation (Variables|/Script/Engine.SceneComponent|GetRelativeLocation _doormesh))
  (Variables|BuildingActor|SetbIsInvulnerable true)
  (Fortnite|SetCanbeDamaged self)
  (Variables|Default|SetDoorStartLoc _relativelocation)
  (Variables|Default|SetDoorEndLoc (+ _relativelocation "0.000000,0.000000,-300.000000")))

(event Custom|OpenDoor
  ;[0.5198 0.1978 0.6865 0.5] Timelines for door opening and door sound
  (Utilities|FlowControl|DoOnce false
    (:Completed
      (bind (_direction _doormovingtrack) (Timeline|TimelineForDoor Play)
        (:Update
          (Transformation|SetRelativeLocation (Variables|Default|GetDoorMesh) (Math|Vector|Lerp(Vector) (Variables|Default|GetDoorStartLoc) (Variables|Default|GetDoorEndLoc) _doormovingtrack) false false))
        (:DoorSound
          (Networking|SwitchHasAuthority
            (:Remote
              (Audio|PlaySoundatLocation "/Game/Sounds/Fort_Doors/World/Door_Bunker_Open_Cue.Door_Bunker_Open_Cue"))))) ; Used the "door soud" event track to play sound as the door opens
    )))

(fn ConstructionScript)
